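; Parent rocket projectile Blueprint for the ToyFox secondary player, handling the rocket's cosmetic trail and audio.
; On BeginPlay, where cosmetics can run, it flags the Niagara trail as local or remote from the instigator's controller and plays fire audio: 2D first-person for the local shooter, attached third-person otherwise.
; A hit stops the in-air audio.
; GetGiantPawn resolves the giant player pawn from the instigator or its attach parent.
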(event EventBeginPlay
  (bind _returnvalue (Game|GetInstigatorController))
  (bind _niagara_trail (Variables|Default|GetNiagara_Trail))
  ;[0.000305 0.07812 0.01168 0.5] BeginPlay: Update the niagara trail FX.
  (|CallParentFunction)
  (Networking|CanExecuteCosmeticEvents
    (:True
      (Utilities|IsValid _returnvalue
        (:Is%20Valid
          (Effects|Components|ParticleSystem|SetBoolParameter _niagara_trail "IsLocal" (Pawn|IsLocalPlayerController _returnvalue)))
        (:Is%20Not%20Valid
          (Effects|Components|ParticleSystem|SetBoolParameter _niagara_trail "IsLocal" false)))
      (CallFunction|PlayFireSound))))

(event Custom|Play%20Fire%20Sound
  (bind _returnvalue (Game|GetInstigatorController))
  ; Rocket fire audio
  (Utilities|IsValid _returnvalue
    (:Is%20Valid
      (if (Pawn|IsLocalPlayerController _returnvalue)
        (Audio|PlaySound2D (Variables|Default|GetFire%20Sound%201p))
        (else
          _)))
    (:Is%20Not%20Valid
      _))
  (Audio|SpawnSoundAttached (Variables|Default|GetFire%20Sound%203p) (Variables|Collision|GetCapsuleComponent)))

(event Collision|EventHit
  (params (MyComp :type Primitive%20Component%20Object%20Reference) (Other :type Actor%20Object%20Reference) (OtherComp :type Primitive%20Component%20Object%20Reference) (bSelfMoved :type Boolean) (HitLocation :type Vector :default "0, 0, 0") (HitNormal :type Vector :default "0, 0, 0") (NormalImpulse :type Vector :default "0, 0, 0") (Hit :type Hit%20Result%20Structure%20%28by%20ref%29))
  ; Muting Audio sound for VFX if an impact occurs.
  (Audio|Components|Audio|Stop (Variables|Default|GetInAirAudio)))

(fn GetGiantPawn
  (bind _returnvalue (Game|GetInstigator))
  (Utilities|FlowControl|Sequence
    (:then_0
      (Utilities|IsValid _returnvalue
        (:Is%20Valid
          (bind _astoy_fox_giant_player_pawn (Utilities|Casting|CastToToyFoxGiantPlayerPawn _returnvalue)
            (:then
              (return _astoy_fox_giant_player_pawn))
            (:CastFailed
              (bind _astoy_fox_giant_player_pawn_2 (Utilities|Casting|CastToToyFoxGiantPlayerPawn (Actor|GetAttachParentActor _returnvalue)))
              (return _astoy_fox_giant_player_pawn_2))))))
    (:then_1
      (return 0))))

(fn ConstructionScript
  (|CallParentFunction))
