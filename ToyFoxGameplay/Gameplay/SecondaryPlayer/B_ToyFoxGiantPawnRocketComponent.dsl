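; Rocket component function for the ToyFox giant pawn that spawns a basic explosive rocket aimed from a spawn point toward a target location.
; Normal requests spawn a client-predicted projectile and place an impact telegraph at the target, aligned to the surface normal.
; Stand-in pilot requests spawn a separate projectile class on authority only, without a telegraph.
; Both use a hotfixable speed and return the rocket, with a Sequence providing an empty fallback return.
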(fn RequestSpawnBasicRocket
  (params (TargetDestinationLocation :type Vector) (RequestedBy :type Actor%20Object%20Reference) (SpawnLoc :type Vector) (TraceImpactNormal :type Vector) (bStandInPilotRequest :type Boolean))
  (Utilities|FlowControl|Sequence
    (:then_0
      (if (not bStandInPilotRequest)
        (bind _returnvalue (Projectile|SpawnClientPredictedProjectile (Variables|Settings|GetBasicRocketProjectileClass) RequestedBy (Utilities|Struct|MakeSpawnProjectileParams SpawnLoc "0.000000,0.000000,0.000000" (Math|Conversions|RotationFromXVector (Math|Vector|GetUnitDirection(Vector) SpawnLoc TargetDestinationLocation)) "(TargetSelection=(List=,bStopAtFirstSuccess=True,bKeepCheckingListOnIndestructibleHit=False,AbilityTargetBlockedGameplayTags=(GameplayTags=),bUseWeaponRanges=True,bUseMaxYawAngleToTarget=False,MaxYawAngleToTarget=0.000000),TargetGameplayEffectSpecs=,OwnerGameplayEffectSpecs=,ActivationCues=(GameplayTags=),ImpactCues=(GameplayTags=),ImpactNoiseRange=0.000000,FlyByNoiseRange=0.000000,bOverrideChargeMagnitude=False,ChargeMagnitudeOverrideValue=0.000000)" "(TargetSelection=(List=,bStopAtFirstSuccess=True,bKeepCheckingListOnIndestructibleHit=False,AbilityTargetBlockedGameplayTags=(GameplayTags=),bUseWeaponRanges=True,bUseMaxYawAngleToTarget=False,MaxYawAngleToTarget=0.000000),TargetGameplayEffectSpecs=,OwnerGameplayEffectSpecs=,ActivationCues=(GameplayTags=),ImpactCues=(GameplayTags=),ImpactNoiseRange=0.000000,FlyByNoiseRange=0.000000,bOverrideChargeMagnitude=False,ChargeMagnitudeOverrideValue=0.000000)" false 0 (ScalableFloat|GetCachedScalableFloatValue (Variables|Hotfix|GetHotfix_ProjectileSpeed)) (ScalableFloat|GetCachedScalableFloatValue (Variables|Hotfix|GetHotfix_ProjectileSpeed)) 0.0)))
        (bind _asb_prj_toy_fox_rocket_explosive_basic (Utilities|Casting|CastToB_Prj_ToyFox_RocketExplosive_Basic _returnvalue))
        (Class|BPrjToyFoxRocketExplosiveBasic|SpawnImpactTelegraph _asb_prj_toy_fox_rocket_explosive_basic TargetDestinationLocation (Math|Rotator|MakeRotfromZ TraceImpactNormal))
        (return _asb_prj_toy_fox_rocket_explosive_basic)
        (else
          (Networking|SwitchHasAuthority
            (:Authority
              (bind _returnvalue_2 (Projectile|SpawnProjectilewithParams (Variables|Settings|GetBasicRocketProjectileClass_StandInPilot) RequestedBy (Utilities|Struct|MakeSpawnProjectileParams (Variables|GetSpawnLoc) "0.000000,0.000000,0.000000" (Math|Conversions|RotationFromXVector (Math|Vector|GetUnitDirection(Vector) (Variables|GetSpawnLoc) (Variables|GetTargetDestinationLocation))) "(TargetSelection=(List=,bStopAtFirstSuccess=True,bKeepCheckingListOnIndestructibleHit=False,AbilityTargetBlockedGameplayTags=(GameplayTags=),bUseWeaponRanges=True,bUseMaxYawAngleToTarget=False,MaxYawAngleToTarget=0.000000),TargetGameplayEffectSpecs=,OwnerGameplayEffectSpecs=,ActivationCues=(GameplayTags=),ImpactCues=(GameplayTags=),ImpactNoiseRange=0.000000,FlyByNoiseRange=0.000000,bOverrideChargeMagnitude=False,ChargeMagnitudeOverrideValue=0.000000)" "(TargetSelection=(List=,bStopAtFirstSuccess=True,bKeepCheckingListOnIndestructibleHit=False,AbilityTargetBlockedGameplayTags=(GameplayTags=),bUseWeaponRanges=True,bUseMaxYawAngleToTarget=False,MaxYawAngleToTarget=0.000000),TargetGameplayEffectSpecs=,OwnerGameplayEffectSpecs=,ActivationCues=(GameplayTags=),ImpactCues=(GameplayTags=),ImpactNoiseRange=0.000000,FlyByNoiseRange=0.000000,bOverrideChargeMagnitude=False,ChargeMagnitudeOverrideValue=0.000000)" false 0 (ScalableFloat|GetCachedScalableFloatValue (Variables|Hotfix|GetHotfix_ProjectileSpeed)) (ScalableFloat|GetCachedScalableFloatValue (Variables|Hotfix|GetHotfix_ProjectileSpeed)) 0.0)))
              (bind _asb_prj_toy_fox_rocket_explosive_basic_2 (Utilities|Casting|CastToB_Prj_ToyFox_RocketExplosive_Basic _returnvalue_2))
              (return _asb_prj_toy_fox_rocket_explosive_basic_2))))))
    (:then_1
      (return 0))))
