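; Gameplay ability that lets a secondary player exit the giant in the ToyFox mode.
; On activation with authority, it gets the Athena game state's B_ToyFoxGameStateComponent and calls HandlePlayerExit with the avatar player pawn.
; It then notifies that component of the secondary pilot exit for analytics.
; The ability ends immediately after the sequence, whether or not the exit was handled.
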(event Ability|EventActivateAbility
  (bind _returnvalue (Fortnite|Athena|GameState|GetGameStateAthena))
  ;[0.06177 0 0.15 0.5] ActivateAbility: Exit the activating player from the giant.
  (|CallParentFunction)
  (Utilities|FlowControl|Sequence
    (:then_0
      (Ability|HasAuthority
        (:True
          (Utilities|IsValid _returnvalue
            (:Is%20Valid
              (bind _returnvalue_2 (Actor|GetComponentbyClass _returnvalue "/ToyFoxGameplay/Gameplay/B_ToyFoxGameStateComponent.B_ToyFoxGameStateComponent_C"))
              (Utilities|IsValid _returnvalue_2
                (:Is%20Valid
                  (Pilots|HandlePlayerExit _returnvalue_2 (CallFunction|GetAvatarasFortPlayerPawn))
                  (Analytics|Notify|NotifySecondaryPilotAbilityExit _returnvalue_2) ; Analytics
                )))))))
    (:then_1
      (Ability|EndAbility))))
